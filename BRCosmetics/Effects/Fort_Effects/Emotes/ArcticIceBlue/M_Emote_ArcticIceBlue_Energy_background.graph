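# Unlit, translucent, two-sided Niagara sprite and mesh material for background energy in the ArcticIceBlue emote.
# Particle color drives emissive, dimmed on lower quality and mobile or forward paths, which also skip depth fade.
# Opacity combines a smoke-distorted scrolling energy texture, rounded Perlin breakup, UV edge masks, view-angle fade, particle alpha and optional dynamic-parameter HeightLerp dissolve.
# Time-animated sines along V also build a local-X wobble offset.
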Material /BRCosmetics/Effects/Fort_Effects/Emotes/ArcticIceBlue/M_Emote_ArcticIceBlue_Energy_background
graph {
 properties {
  bUsedWithNiagaraMeshParticles=true
  bUsedWithNiagaraSprites=true
  blendMode=BLEND_Translucent
  pixelDepthOffsetMode=PDOM_Legacy
  shadingModel=MSM_Unlit
  twoSided=true
  usedShadingModels=Unlit
 }
 nodes [
  n1 = TextureCoordinate()
  n2 = Time()
  n3 = Sine(None=Add(A=ComponentMask(None=@n1) {g=true}, B=Multiply(A=@n2) {constB=2}))
  n4 = TextureCoordinate()
  n5 = Time()
  n6 = Sine(None=Add(A=ComponentMask(None=@n4) {g=true}, B=Multiply(A=@n5))) {period=0.8}
  n7 = Transform(None=Multiply(A=LinearInterpolate(A=@n3, B=@n6), B=Constant3Vector() {constant=<rgba 1 0 0 0>})) {transformSourceType=TRANSFORMSOURCE_Local}
  n8 = ComponentMask(None=TextureCoordinate()) {g=true}
  Add(A=Multiply(A=Multiply(A=@n7) {constB=5}, B=OneMinus(None=@n8))) {constB=0}
  n9 = ParticleColor()
  n10 = Multiply(A=@n9.RGB) {constB=0.7}
  n11 = QualitySwitch(Default=@n9.RGB, Low=@n10, High=@n9.RGB, Medium=@n10, Epic=@n9.RGB)
  MaterialFunctionCall(Emissive=ShadingPathSwitch(Default=@n11, Deferred=@n11, Forward=@n10, Mobile=@n10)) {materialFunction=&MF_FX_EmissiveAdjustments}
  MaterialFunctionCall() {materialFunction=&DebugTimeSine}
  n12 = Panner(Coordinate=TextureCoordinate() {vTiling=0.5}) {speedY=1.4}
  n13 = TextureCoordinate()
  n14 = Multiply(A=TextureSample(UVs=Panner(Coordinate=@n13) {speedY=1}) {texture=&T_Inky_Smoke_Tile}.R) {constB=0.15}
  n15 = Sine(None=ComponentMask(None=TextureCoordinate()) {g=true}) {period=2}
  n16 = Multiply(
   A=TextureSample(UVs=Add(A=@n12, B=@n14)) {texture=&/T_Emote_ArcticIceBlue_Energy_background}.R
   B=Saturate(None=Multiply(A=@n15) {constB=3})
  )
  n17 = DepthFade(Opacity=@n16) {fadeDistanceDefault=50}
  n18 = Panner(Coordinate=Add(A=TextureCoordinate() {uTiling=0.5, vTiling=0.5}) {constB=0}) {speedY=1}
  n19 = ComponentMask(None=TextureCoordinate()) {g=true}
  n20 = LinearInterpolate(
   A=ScalarParameter() {defaultValue=-1, parameterName=Dissolve mask Min}
   B=Constant() {r=1}
   Alpha=Power(Base=OneMinus(None=@n19)) {constExponent=1.3}
  ) {constA=0.1, constB=2}
  n21 = Round(None=Power(
    Base=TextureSample(UVs=@n18) {samplerType=SAMPLERTYPE_LinearGrayscale, texture=&T_SimplePerlin_Greyscale}.R
    Exp=Saturate(None=@n20)
   ))
  n22 = DynamicParameter() {paramNames [Param1,Param2,Dissolve,Param4]}
  n23 = TextureSample(UVs=Panner(Coordinate=TextureCoordinate() {uTiling=0.7}) {speedY=0.5}) {texture=/BRCosmetics/Weapons/FORT_Melee/Pickaxe_MadameMoth/FX/T_Fibre_Noise}
  n24 = Multiply(
   A=Multiply(A=ShadingPathSwitch(Default=@n17, Deferred=@n17, Forward=@n16, Mobile=@n16), B=Saturate(None=@n21))
   B=LinearInterpolate(
    B=MaterialFunctionCall(Transition Phase=@n22.Dissolve, Height Texture=@n23.R) {materialFunction=&HeightLerp}.Alpha
    Alpha=ScalarParameter() {parameterName=Use Dissolve}
   ) {constA=1}
  )
  n25 = ComponentMask(None=TextureCoordinate()) {g=true}
  n26 = Saturate(None=Multiply(A=Power(Base=@n25)) {constB=20})
  Multiply(
   A=MaterialFunctionCall("FadeSensitivity (optional)"=Constant() {r=2}) {materialFunction=&FadeBasedOnViewAngle}.Result
   B=Multiply(A=Multiply(A=@n24, B=@n26), B=@n9.A)
  )
 ]
}
