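# Unlit translucent material for Niagara mesh particles, forming the tornado in the ArcticIceBlue emote effect.
# Two stretched, panning noise layers (plasma and Perlin) are rounded into hard-edged swirl bands that modulate particle color.
# Opacity combines particle alpha, V-axis end fades and a noise-varied view-angle fade.
# Quality and shading-path switches dim emissive on lower tiers, forward and mobile, and omit depth fade on forward and mobile.
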Material /BRCosmetics/Effects/Fort_Effects/Emotes/ArcticIceBlue/M_Emote_ArcticIceBlue_Tonado
graph {
 properties {
  bUsedWithNiagaraMeshParticles=true
  blendMode=BLEND_Translucent
  pixelDepthOffsetMode=PDOM_Legacy
  shadingModel=MSM_Unlit
  usedShadingModels=Unlit
 }
 nodes [
  n1 = ParticleColor()
  n2 = MaterialFunctionCall(UVs=Panner(Coordinate=TextureCoordinate()) {speedX=-3, speedY=1}, "Rotation Angle (0-1)"=Constant()) {materialFunction=&CustomRotator}
  n3 = Constant2Vector() {g=2, r=0.2}
  n4 = Power(Base=TextureSample(UVs=Multiply(A=@n2.Rotated Values, B=@n3)) {samplerType=SAMPLERTYPE_LinearGrayscale, texture=/Game/Effects/Fort_Effects/Emotes/UltraLight/Texture/T_Plasma_C_Greyscale}.R) {constExponent=0.8}
  n5 = TextureCoordinate()
  n6 = Multiply(
   A=MaterialFunctionCall(UVs=Panner(Coordinate=@n5) {speedX=-2.5, speedY=0.3}, "Rotation Angle (0-1)"=Constant()) {materialFunction=&CustomRotator}.Rotated Values
   B=Constant2Vector() {g=2.5, r=0.3}
  )
  n7 = Round(None=Power(Base=TextureSample(UVs=@n6) {samplerType=SAMPLERTYPE_LinearGrayscale, texture=&T_SimplePerlin_Greyscale}.R) {constExponent=0.9})
  n8 = Multiply(A=Add(A=Round(None=@n4), B=Multiply(A=@n7) {constB=0.6}))
  n9 = Multiply(A=@n1.RGB, B=LinearInterpolate(A=ScalarParameter() {defaultValue=0.2, parameterName=Dark}, Alpha=Multiply(A=@n8)) {constA=0.2})
  n10 = Multiply(A=@n9) {constB=0.7}
  n11 = QualitySwitch(Default=@n9, Low=@n10, High=@n9, Medium=@n10, Epic=@n9)
  MaterialFunctionCall(Emissive=ShadingPathSwitch(Default=@n11, Deferred=@n11, Forward=@n10, Mobile=@n10)) {materialFunction=&MF_FX_EmissiveAdjustments}
  n12 = TextureSample(UVs=TextureCoordinate() {uTiling=0, vTiling=3}) {samplerType=SAMPLERTYPE_Grayscale, texture=&LowResBlurredNoise}
  n13 = MaterialFunctionCall("FadeSensitivity (optional)"=Multiply(A=Constant() {r=2}, B=LinearInterpolate(Alpha=@n12.R) {constA=0.5, constB=5})) {materialFunction=&FadeBasedOnViewAngle}
  Multiply(A=@n13.Result)
  n14 = ComponentMask(None=TextureCoordinate()) {g=true}
  n15 = Saturate(None=Multiply(A=Power(Base=@n14) {constExponent=1}) {constB=2})
  n16 = OneMinus(None=ComponentMask(None=TextureCoordinate()) {g=true})
  n17 = Saturate(None=Multiply(A=Subtract(A=@n16) {constB=0.05}) {constB=5})
  n18 = Multiply(A=@n1.A, B=Multiply(A=Multiply(A=@n15, B=@n17), B=@n13.Result))
  n19 = DepthFade(Opacity=@n18) {fadeDistanceDefault=50}
  ShadingPathSwitch(Default=@n19, Deferred=@n19, Forward=@n18, Mobile=@n18)
 ]
}
