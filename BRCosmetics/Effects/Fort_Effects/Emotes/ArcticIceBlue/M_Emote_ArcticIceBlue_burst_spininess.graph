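# Translucent unlit Niagara sprite material for the spiny burst effect in the ArcticIceBlue emote.
# Particle color drives emissive, dimmed on lower quality and on Forward or Mobile paths.
# A panning, tapered noise distorts the burst texture, whose masked alpha is shaped by vertical gradients and view-angle fade.
# Dynamic parameters drive a noise-based HeightLerp dissolve, and depth fade softens opacity on Default and Deferred paths.
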Material /BRCosmetics/Effects/Fort_Effects/Emotes/ArcticIceBlue/M_Emote_ArcticIceBlue_burst_spininess
graph {
 properties {bUsedWithNiagaraSprites=true, blendMode=BLEND_Translucent, pixelDepthOffsetMode=PDOM_Legacy, shadingModel=MSM_Unlit, usedShadingModels=Unlit}
 nodes [
  n1 = ParticleColor()
  n2 = Multiply(A=@n1.RGB, B=ScalarParameter() {defaultValue=0.3, parameterName=Low Bright}) {constB=0.3}
  n3 = QualitySwitch(Default=@n1.RGB, Low=@n2, High=@n1.RGB, Medium=@n2, Epic=@n1.RGB)
  Reroute(None=MaterialFunctionCall(Emissive=ShadingPathSwitch(Default=@n3, Deferred=@n3, Forward=@n2, Mobile=@n2)) {materialFunction=&MF_FX_EmissiveAdjustments}.Result)
  n4 = TextureCoordinate() {uTiling=4}
  n5 = ComponentMask(None=@n4) {g=true}
  n6 = Multiply(A=Subtract(A=ComponentMask(None=@n4) {r=true}) {constB=1.7}, B=LinearInterpolate(Alpha=@n5) {constA=1, constB=3})
  n7 = TextureSample(UVs=Panner(Coordinate=AppendVector(A=@n6, B=@n5)) {speedY=2}) {
   samplerType=SAMPLERTYPE_LinearGrayscale
   texture=/Game/Characters/Player/Male/Medium/Bodies/M_MED_Citadel/FX/T_Citadel_Noise_Dissolve_Handpaint
  }
  n8 = TextureCoordinate()
  n9 = Add(A=OneMinus(None=ComponentMask(None=@n8) {g=true}), B=ScalarParameter() {parameterName=Mask offset})
  n10 = Multiply(A=@n7.R, B=Saturate(None=Multiply(A=@n9) {constB=2}))
  n11 = TextureSampleParameter2D(UVs=Add(A=TextureCoordinate(), B=Multiply(A=@n10) {constB=0.1}) {constB=0}) {parameterName=Default Tex, texture=&/T_Emote_ArcticIceBlue_burst_spininess}
  n12 = ComponentMask(None=TextureCoordinate()) {g=true}
  n13 = Multiply(
   A=Multiply(A=@n1.A, B=StaticComponentMaskParameter(None=@n11.RGB) {defaultR=true, parameterName=mask})
   B=Saturate(None=Multiply(A=@n12) {constB=5})
  )
  n14 = TextureCoordinate()
  n15 = Power(Base=OneMinus(None=ComponentMask(None=@n14) {g=true})) {constExponent=3}
  n16 = LinearInterpolate(B=Saturate(None=Multiply(A=@n15) {constB=40}), Alpha=ScalarParameter() {parameterName=Use bottom mask}) {constA=1}
  n17 = ScalarParameter() {defaultValue=2, parameterName=ViewAngleFadeSensitivity}
  n18 = DynamicParameter() {paramNames [Dissolve,Param2,ErosionX,ErosionY]}
  n19 = DynamicParameter() {paramNames [Dissolve,Param2,ErosionX,ErosionY]}
  n20 = Multiply(
   A=TextureCoordinate()
   B=ComponentMask(None=VectorParameter() {parameterName=Mask UV Scale, defaultValue=<rgba 1 1 0 1>}.RGB) {g=true, r=true}
  )
  n21 = ComponentMask(None=@n20) {g=true}
  n22 = Multiply(
   A=Subtract(A=ComponentMask(None=@n20) {r=true}, B=ScalarParameter() {defaultValue=0.45, parameterName=Mask offset02}) {constB=0.45}
   B=LinearInterpolate(Alpha=@n21) {constA=1, constB=3}
  )
  n23 = TextureSample(UVs=Add(A=AppendVector(A=@n19.ErosionX, B=@n19.ErosionX), B=AppendVector(A=@n22, B=@n21))) {
   samplerType=SAMPLERTYPE_LinearGrayscale
   texture=&T_Citadel_Noise_Dissolve_Handpaint
  }
  n24 = Multiply(
   A=Multiply(
    A=Multiply(A=@n13, B=@n16)
    B=MaterialFunctionCall("FadeSensitivity (optional)"=@n17) {materialFunction=&FadeBasedOnViewAngle}.Result
   )
   B=LinearInterpolate(Alpha=MaterialFunctionCall(Transition Phase=@n18.Dissolve, Height Texture=@n23.R) {materialFunction=&HeightLerp}.Alpha) {constA=-8, constB=10}
  )
  n25 = DepthFade(Opacity=@n24, FadeDistance=ScalarParameter() {defaultValue=50, parameterName=FadeDistance}) {fadeDistanceDefault=50}
  ShadingPathSwitch(Default=@n25, Deferred=@n25, Forward=@n24, Mobile=@n24)
 ]
}
