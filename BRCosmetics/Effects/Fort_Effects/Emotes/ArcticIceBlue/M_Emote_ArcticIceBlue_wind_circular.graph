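# Translucent, unlit, two-sided particle material for the circular wind effect of the ArcticIceBlue emote, usable on Niagara mesh particles.
# A UV-scaled, channel-selectable texture mask multiplies particle alpha for opacity and can optionally shape the particle-colored emissive.
# Swirl noise through HeightLerp, driven by the Dissolve dynamic parameter and rounded, gives a hard-edged erosion.
# View-angle fading softens edges, and depth fade applies except on Forward and Mobile shading paths.
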Material /BRCosmetics/Effects/Fort_Effects/Emotes/ArcticIceBlue/M_Emote_ArcticIceBlue_wind_circular
graph {
 properties {
  bUsedWithNiagaraMeshParticles=true
  blendMode=BLEND_Translucent
  pixelDepthOffsetMode=PDOM_Legacy
  shadingModel=MSM_Unlit
  twoSided=true
  usedShadingModels=Unlit
 }
 nodes [
  n1 = TextureCoordinate()
  n2 = ComponentMask(None=VectorParameter() {parameterName=UV Scale, defaultValue=<rgba 1 1 0 0>}.RGB) {g=true, r=true}
  n3 = StaticComponentMaskParameter(None=TextureSampleParameter2D(UVs=Multiply(A=@n1, B=@n2)) {parameterName=Default Tex, texture=&/T_Emote_ArcticIceBlue_wind_circular}.RGB) {defaultR=true, parameterName=Component mask}
  n4 = ScalarParameter() {parameterName=Use shape mask}
  n5 = ParticleColor()
  MaterialFunctionCall(Emissive=Multiply(A=LinearInterpolate(B=@n3, Alpha=@n4) {constA=1}, B=@n5.RGB) {constA=1}) {materialFunction=&MF_FX_EmissiveAdjustments}
  n6 = Add(A=TextureCoordinate(), B=Constant2Vector())
  n7 = MaterialFunctionCall(
   Transition Phase=DynamicParameter() {paramNames [Dissolve,Param2,Param3,Param4]}.Dissolve
   Height Texture=Multiply(A=TextureSample(UVs=@n6) {samplerType=SAMPLERTYPE_LinearGrayscale, texture=/Game/Global/Textures/Default/Noise/T_SwirlNoise_N_height}.R)
   Contrast=ScalarParameter() {defaultValue=0.5, parameterName="Contrast (S)"}
  ) {materialFunction=&HeightLerp}
  n8 = Multiply(
   A=Multiply(A=Multiply(A=@n5.A, B=@n3), B=Round(None=@n7.Alpha))
   B=MaterialFunctionCall("FadeSensitivity (optional)"=ScalarParameter() {parameterName=Edge fade}) {materialFunction=&FadeBasedOnViewAngle}.Result
  )
  n9 = DepthFade(Opacity=@n8) {fadeDistanceDefault=30}
  ShadingPathSwitch(Default=@n9, Deferred=@n9, Forward=@n8, Mobile=@n8)
 ]
}
